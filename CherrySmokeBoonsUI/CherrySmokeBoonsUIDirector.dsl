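; Dynamic UI director that shows and hides Cherry Smoke boon HUD scenes for the local player's view target.
; On view-target change it binds augment updates, which toggle the rank backing through a delta gate.
; Outside mobile touch, async gameplay-tag waits toggle the Hot Hands boon and Action Reload scenes.
; A scalable-float hotfix adds the bottom-screen rank-up scene, and EndPlay adds it again despite its "Remove" comment.
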(event EventBeginPlay
  (|CallParentFunction)
  (Utilities|FlowControl|Sequence
    (:then_0
      ;[1 0.5772 0 0.5] HOTFIX
      ; If true, use bottom screen message widget
      ; If false, use Hero Ranking Badge widget notifications
      (if (CallFunction|IsRankUpNotifHotfixActive)
        (DynamicUIDirector|AddScene (Variables|DynamicUIScene|GetDynamicUIScene_RankUpNotif_BottomScreen))))
    (:then_1
      (bind _asathena_player_controller (Utilities|Casting|CastToAthena_PlayerController (DynamicUIDirector|GetOwningLocalPlayerController)))
      (EventDispatchers|BindEventtoOnViewTargetChanged _asathena_player_controller (EventDispatchers|CreateEvent :!event Custom|OnViewTargetChanged))
      (CallFunction|OnViewTargetChanged 0 0 (Pawn|GetControlledPawn _asathena_player_controller)))))

(event Custom|OnViewTargetChanged
  (params (PlayerController :type Fort%20Player%20Controller%20Object%20Reference) (OldViewTarget :type Actor%20Object%20Reference) (NewViewTarget :type Actor%20Object%20Reference))
  ; On View Target Changed
  (Utilities|IsValid NewViewTarget
    (:Is%20Valid
      (bind _asfort_player_pawn (Utilities|Casting|CastToFortPlayerPawn NewViewTarget))
      (bind _asfort_player_state_athena (Utilities|Casting|CastToFortPlayerStateAthena (Variables|/Script/Engine.Pawn|GetPlayerState _asfort_player_pawn)))
      (Variables|Default|SetPlayer%20State%20Ref _asfort_player_state_athena)
      (EventDispatchers|BindEventtoOnAugmentListUpdated (Class|FortPlayerStateComponentPlayerAugmentSystem|GetAugmentComponent (Variables|/Script/Engine.Pawn|GetPlayerState _asfort_player_pawn)) (EventDispatchers|CreateEvent :!event HandleMedallionsInInventoryChange))
      (if (not (and (TouchPlatform|IsMobileGame) (TouchPlatform|IsUsingTouch (DynamicUIDirector|GetOwningLocalPlayerController))))
        (CallFunction|WaitforPlayerHotHandsTag NewViewTarget)
        (CallFunction|WaitforPlayerActionReload NewViewTarget)))))

(event EventEndPlay
  (params (EndPlayReason :type EEndPlayReason%20Enum :default "Destroyed"))
  (|CallParentFunction EndPlayReason)
  (DynamicUIDirector|RemoveScene (Variables|DynamicUIScene|GetDynamicUIScene_BoonsRankBacking))
  ;[1 0.5772 0 0.5] HOTFIX
  ; Remove scene if active
  (if (CallFunction|IsRankUpNotifHotfixActive)
    (DynamicUIDirector|AddScene (Variables|DynamicUIScene|GetDynamicUIScene_RankUpNotif_BottomScreen))))

(event Custom|WaitForPlayerHotHandsTag
  (params (PlayerActor :type Actor%20Object%20Reference))
  (bind _returnvalue (DynamicUIDirector|GetOwningLocalPlayerController))
  ;[1 1 1] Wait for Action Reload Event (Acrobat Combat Boon Equipped) (this May Not be needed)
  (Utilities|IsValid PlayerActor
    (:Is%20Valid
      (Utilities|IsValid _returnvalue
        (:Is%20Valid
          (Utilities|Casting|CastToFortPlayerPawnAthena (Pawn|GetControlledPawn _returnvalue))
          (Ability|Tasks|WaitGameplayTagAddToActor PlayerActor (Variables|Default|GetHotHandsTag_Event_0)
            (:then
              (Ability|Tasks|WaitGameplayTagRemoveFromActor PlayerActor (Variables|Default|GetHotHandsTag_Event_0)
                (:Removed
                  (DynamicUIDirector|RemoveScene (Variables|DynamicUIScene|GetDynamicUIScene_CherrySmokeBoons)))))
            (:Added
              (DynamicUIDirector|AddScene (Variables|DynamicUIScene|GetDynamicUIScene_CherrySmokeBoons)))))))))

(event Custom|WaitForPlayerActionReload
  (params (PlayerActor :type Actor%20Object%20Reference))
  (bind _returnvalue (DynamicUIDirector|GetOwningLocalPlayerController))
  ;[0 0 0] Wait to Add Action Reload Scene For when the player has either the Boon or the Mask Maker Medallion
  (Utilities|IsValid PlayerActor
    (:Is%20Valid
      (Utilities|IsValid _returnvalue
        (:Is%20Valid
          (Utilities|Casting|CastToFortPlayerPawnAthena (Pawn|GetControlledPawn _returnvalue))
          (Ability|Tasks|WaitGameplayTagAddToActor PlayerActor (Variables|Default|GetActionReloadTag_Event)
            (:then
              (Ability|Tasks|WaitGameplayTagRemoveFromActor PlayerActor (Variables|Default|GetActionReloadTag_Event)
                (:Removed
                  (DynamicUIDirector|RemoveScene (Variables|Default|GetDynamicUIScene_ActionReload)))))
            (:Added
              (DynamicUIDirector|AddScene (Variables|Default|GetDynamicUIScene_ActionReload)))))))))

(fn HandleMedallionsInInventoryChange
  (params (AugmentComponent :type Fort%20Player%20State%20Component%20Player%20Augment%20System%20Object%20Reference))
  (bind _returnvalue (Class|FortPlayerStateComponentPlayerAugmentSystem|GetActiveAugments (Variables|Default|GetPlayer%20State%20Ref)))
  (bind _updated_variable (Utilities|FlowControl|DeltaGateWildcard (Variables|Default|GetIsMedallionSceneActive) (> (Utilities|Array|Length _returnvalue) 0))
    (:Success
      (if _updated_variable
        (DynamicUIDirector|AddScene (Variables|DynamicUIScene|GetDynamicUIScene_BoonsRankBacking))
        (else
          (DynamicUIDirector|RemoveScene (Variables|DynamicUIScene|GetDynamicUIScene_BoonsRankBacking)))))))

(fn IsRankUpNotifHotfixActive
  (return (>= (ScalableFloat|GetValueatLevel (Variables|Default|GetSF_ShouldUseBottomScreenNotification)) 1.0)))

(fn ConstructionScript)
