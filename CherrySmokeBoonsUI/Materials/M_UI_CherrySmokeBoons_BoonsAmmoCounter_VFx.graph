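# M_UI_CherrySmokeBoons_BoonsAmmoCounter_VFx is a translucent UI material for a green animated progress effect on the Cherry Smoke Boons ammo counter.
# Time scrolls soft horizontal stripes that vary the green color.
# Radial arc functions mask a circle outline into meter, delta and background layers, while start, end and animate parameters drive horizontal pip progress.
# An isRadial scalar blends linear and radial masks, scaled by Opacity.
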Material /CherrySmokeBoonsUI/Materials/M_UI_CherrySmokeBoons_BoonsAmmoCounter_VFx
graph {
 properties {bAutomaticallySetUsageInEditor=false, blendMode=BLEND_Translucent, materialDomain=MD_UI, usedShadingModels=Default Lit}
 nodes [
  n1 = Time()
  n2 = ScalarParameter() {defaultValue=2.5, parameterName=Speed}
  n3 = ComponentMask(None=OneMinus(None=TextureCoordinate())) {g=true}
  n4 = ScalarParameter() {defaultValue=100, parameterName=Number of Lines}
  n5 = Frac(None=Add(A=Multiply(A=@n1, B=@n2), B=Multiply(A=@n3, B=@n4)))
  n6 = SmoothStep(
   Min=ScalarParameter() {defaultValue=-0.0691, parameterName=Line Thickness}
   Max=ScalarParameter() {defaultValue=0.4, parameterName=Line Blur}
   Value=Abs(None=Subtract(A=@n5) {constB=0.5})
  )
  n7 = Multiply(A=@n6) {constB=0.25}
  LinearInterpolate(
   A=Multiply(A=OneMinus(None=@n7), B=Constant4Vector() {constant=<rgba 0.1499 0.3826 0.1942 1>}.RGBA) {constB=0.1}
   B=NamedRerouteUsage()
   Alpha=NamedRerouteUsage()
  )
  n8 = Multiply(A=NamedRerouteUsage(), B=NamedRerouteUsage())
  n9 = MaterialFunctionCall(UVs=MaterialFunctionCall!(UV, Offset, Rotation, Scale, MirrorY) {materialFunction=/FortUILibrary/MaterialFunctions/Basics/MF_UI_Symmetry_2x1_DEPRECATED}.Result) {materialFunction=&MF_UI_LinearGradient}
  n10 = NamedRerouteUsage()
  n11 = NamedRerouteUsage()
  Multiply(
   A=LinearInterpolate(A=Multiply(A=@n8, B=@n9.Result), B=Multiply(A=@n10, B=@n11), Alpha=ScalarParameter() {defaultValue=1, parameterName=isRadial})
   B=ScalarParameter() {defaultValue=1, parameterName=Opacity}
  )
  n12 = ScalarParameter() {defaultValue=0.25, group=Radial, parameterName=MaskStartAngle}
  n13 = ScalarParameter() {defaultValue=0.25, group=Radial, parameterName=MaskEndAngle}
  n14 = ScalarParameter() {
   defaultValue=1
   desc="Property for animating progress value change (from ProgressStart to ProgressEnd)\r\n0 - 1"
   group=Progress
   parameterName=AnimatePipProgress
   sliderMax=1
   sortPriority=3
  }
  n15 = OneMinus(None=@n14)
  n16 = ScalarParameter() {defaultValue=1, group=Radial, parameterName=FillClockwise?}
  n17 = MaterialFunctionCall!(UV, "Arc Start Angle (0-1)"=OneMinus(None=@n13), "Arc End Angle (0-1)"=OneMinus(None=@n12), "Fill Ratio (0-1)"=@n15, Fill Clockwise=@n16) {materialFunction=/FortUILibrary/MaterialFunctions/Basics/MF_UI_RadialArc}
  NamedRerouteDeclaration(None=Min(
    A=MaterialFunctionCall("Arc Start Angle (0-1)"=@n12, "Arc End Angle (0-1)"=@n13, "Fill Ratio (0-1)"=@n15, Fill Clockwise=@n16) {materialFunction=&MF_UI_RadialArc}.Radial Fill
    B=Multiply(A=@n17.Radial Fill) {constB=1.5}
   )) {name=RadialMask, nodeColor=<rgba 0 0 0 1>}
  n18 = MaterialFunctionCall(
   Size=ScalarParameter() {defaultValue=0.8, group=Circle Parameters, parameterName=CircleSize, sliderMax=1, sortPriority=4}
   OutlineThickness=ScalarParameter() {defaultValue=0.13, group=Circle Parameters, parameterName=CircleThickness, sliderMax=1, sortPriority=4}
  ) {materialFunction=&MF_UI_Circle}
  n19 = MaterialFunctionCall(
   UVs=MaterialFunctionCall() {materialFunction=&GetUserInterfaceUV}.Normalized UV
   "Rotation Angle (0-1)"=ScalarParameter() {group=Radial Parameters, parameterName=MeterRotation, sliderMax=1, sortPriority=3}
  ) {materialFunction=&CustomRotator}
  n20 = ScalarParameter() {defaultValue=0.25, group=Circle Parameters, parameterName=StartAngle, sliderMax=1, sortPriority=4}
  n21 = ScalarParameter() {defaultValue=0.487, group=Circle Parameters, parameterName=EndAngle, sliderMax=1, sortPriority=4}
  n22 = ScalarParameter() {group=Circle Parameters, parameterName=FillDirection, sliderMax=1, sliderMin=-1, sortPriority=4}
  n23 = MaterialFunctionCall(
   UV=@n19.Rotated Values
   "Arc Start Angle (0-1)"=@n20
   "Arc End Angle (0-1)"=@n21
   "Fill Ratio (0-1)"=OneMinus(None=ScalarParameter() {defaultValue=1, group=Progress Parameters, parameterName=MeterProgress, sliderMax=1, sortPriority=1})
   Fill Clockwise=@n22
  ) {materialFunction=&MF_UI_RadialArc}
  n24 = Ceil(None=@n23.Arc Mask)
  n25 = Multiply(A=@n18.Outline, B=@n24)
  n26 = Constant() {r=0.001}
  n27 = NamedRerouteDeclaration(None=MaterialFunctionCall(DistanceField=@n23.Radial Fill, Threshold=@n26) {materialFunction=&MF_UI_AntiAliasing}.Outer) {name=radialMeterFill, nodeColor=<rgba 0.025 0.0203 0.0125 0.75>}
  n28 = MaterialFunctionCall(
   UV=@n19.Rotated Values
   "Arc Start Angle (0-1)"=@n20
   "Arc End Angle (0-1)"=@n21
   "Fill Ratio (0-1)"=OneMinus(None=ScalarParameter() {group=Progress Parameters, parameterName=DeltaProgress, sliderMax=1, sortPriority=1})
   Fill Clockwise=@n22
  ) {materialFunction=&MF_UI_RadialArc}
  n29 = NamedRerouteDeclaration(None=MaterialFunctionCall(DistanceField=@n28.Radial Fill, Threshold=@n26) {materialFunction=&MF_UI_AntiAliasing}.Outer) {name=radialDeltaFill, nodeColor=<rgba 0.025 0.0203 0.0125 0.75>}
  n30 = Saturate(None=Ceil(None=OneMinus(None=@n28.Arc Mask)))
  n31 = ScalarParameter() {defaultValue=0.5, parameterName=BackgroundOpacity, sliderMax=1, sortPriority=4}
  n32 = Multiply(A=@n25, B=Max(A=Max(A=@n27, B=@n29), B=Multiply(A=@n30, B=@n31) {constB=0.5}))
  n33 = Multiply(
   A=MaterialFunctionCall!(DistanceField=@n18.SDF, Threshold, Thickness=Constant() {r=0.08}, Alignment, Smoothing) {materialFunction=/FortUILibrary/MaterialFunctions/Shapes/Modifiers/MF_UI_ShapeOutline}.Outline
   B=@n24
  )
  n34 = Constant()
  n35 = Constant()
  n36 = Multiply(
   A=Min(A=@n25, B=OneMinus(None=@n33))
   B=Multiply(A=MaterialFunctionCall(UV=@n19.Rotated Values, "Arc Start Angle (0-1)"=@n20, "Arc End Angle (0-1)"=@n21, "Fill Ratio (0-1)"=@n34, Fill Clockwise=@n35) {materialFunction=&MF_UI_RadialArc}.Centered Fill) {constB=2}
  )
  n37 = Multiply(A=@n36) {constB=0.1}
  NamedRerouteDeclaration(None=Multiply(A=Add(A=@n32, B=@n37), B=ScalarParameter() {defaultValue=2, parameterName=MasterOpacity, sliderMax=1})) {name=Layers, nodeColor=<rgba 0 0 0 1>}
  n38 = TextureCoordinate()
  n39 = Constant()
  n40 = ScalarParameter() {
   desc="When progress value changes, it defines where change starts.\r\n0 - 1"
   group=Progress
   parameterName=PipProgressStart
   sliderMax=1
   sortPriority=1
  }
  n41 = ScalarParameter() {
   defaultValue=1
   desc="When progress value changes, it defines where change ends.\r\n0 - 1"
   group=Progress
   parameterName=PipProgressEnd
   sliderMax=1
   sortPriority=2
  }
  n42 = ScalarParameter() {
   defaultValue=1
   desc="Property for animating progress value change (from ProgressStart to ProgressEnd)\r\n0 - 1"
   group=Progress
   parameterName=AnimatePipProgress
   sliderMax=1
   sortPriority=3
  }
  n43 = Add(
   A=Multiply(A=MaterialFunctionCall(UVs=@n38, "Rotation Angle (0-1)"=@n39) {materialFunction=&CustomRotator}.Rotated Values) {constB=-1}
   B=NamedRerouteDeclaration(None=LinearInterpolate(A=@n40, B=@n41, Alpha=@n42)) {name=Progress_1, nodeColor=<rgba 0 0 0 1>}
  ) {constB=0.5}
  NamedRerouteDeclaration(None=Saturate(None=ComponentMask(None=@n43) {r=true})) {name=ProgressGraident_Horizontal, nodeColor=<rgba 0 0 0 1>}
  n44 = Constant4Vector() {constant=<rgba 0 0.5781 0 0>}
  n45 = Constant4Vector() {constant=<rgba 0.2341 0.7135 0.2398 1>}
  NamedRerouteDeclaration(None=LinearInterpolate(A=LinearInterpolate(A=@n44.RGBA, B=@n45.RGBA, Alpha=@n6), B=Constant4Vector() {constant=<rgba 0.4504 0.9826 0.5079 1>}.RGBA) {constAlpha=0.3}) {name=ProgressColor, nodeColor=<rgba 0.025 0.0203 0.0125 0.75>}
  n46 = NamedRerouteUsage()
  NamedRerouteDeclaration(None=Add(A=Multiply(A=@n36, B=@n46), B=@n46)) {name=ColorAlphaLerp, nodeColor=<rgba 0 0 0 1>}
 ]
}
